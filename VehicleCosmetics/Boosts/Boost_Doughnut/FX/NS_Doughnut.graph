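# Niagara system for the Doughnut vehicle boost cosmetic, trailing tumbling donuts, sprinkles and a short splash from the car's boost sockets.
# Donut and splash emitters mix meshes with sprites; two near-identical sprinkle emitters, one GPU-simulated, render sprites.
# Each emitter partitions particles across filtered skeletal mesh sockets, applies painted-variant colors and picks backward velocity by GameModeID.
# ColorVariantID tunes donut shading, while bIsFrontendPreview adjusts splash size.
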NiagaraSystem /VehicleCosmetics/Boosts/Boost_Doughnut/FX/NS_Doughnut
dependencies {
 usedDataInterfaces [
  &NiagaraDataInterfaceVectorCurve
  &NiagaraDataInterfaceCurve
  &NiagaraDataInterfaceVector2DCurve
  &NiagaraDataInterfaceArrayColor
  &NiagaraDataInterfaceArrayFloat
  &NiagaraDataInterfaceArrayFloat3
  &NiagaraDataInterfaceSkeletalMesh
 ]
 usedDynamicInputs []
 usedModules [
  &self:SystemSpawnScript.SystemScriptSource.SystemScriptGraph.NiagaraNodeAssignment_0.SetVariables_#1
  &SystemState
 ]
}
emitters {
 NE_VC_Donuts {
  emitterSpawnScript [/SetVariables(
    Emitter.breadlight:NiagaraFloat=&SelectFloatFromArray(
     Direct Array Index=$User.ColorVariantID
     Float Selection Array=DataInterface() {propertyValues {FloatData [0.45,0.45,0.5,0.5,0.45,0.45,0.5,0.5,0.45,0.48,0.5,0.5,0.5,0.5,0.5]}}
    )
    Emitter.sprinklessaturation:NiagaraFloat=&SelectFloatFromArray(
     Direct Array Index=$User.ColorVariantID
     Float Selection Array=DataInterface() {propertyValues {FloatData [1,1,1,0.7,0.7,0.85,0.85,0.85,0.95,0.9,0.9,0.95,0.85,0.85,0.9]}}
    )
   ) {isSetParameters=true}]
  emitterUpdateScript [
   &EmitterState()
   &VC_FilteredSocketCount()
   &SpawnRate(SpawnRate=20, Use Spawn Probability=false)
  ]
  particleSpawnScript [
   &VC_PaintedVariantsOverride(
    Black=<rgba 0 0 0 1>
    BurntSienna=<rgba 0.24 0.0552 0.0384 1>
    Cobalt=<rgba 0.085 0.1487 0.85 1>
    Crimson=<rgba 0.53 0.0159 0.0159 1>
    DefaultColor=&SelectLinearColorFromArray(
     Array Sampling Mode=::NewEnumerator0
     Direct Array Index=?
     Color Selection Array=DataInterface() {propertyValues {ColorData [<rgba 0.7865 0.127 0.7807 1>, <rgba 0.127 0.7039 0.7865 1>, <rgba 0.786 0.786 0.786 1>, <rgba 0.7865 0.127 0.7807 1>]}}
     Random Seed=0
    )
    ForestGreen=<rgba 0.0135 0.45 0.0135 1>
    Grey=<rgba 0.27 0.27 0.27 1>
    Lime=<rgba 0.5628 0.84 0.0084 1>
    Orange=<rgba 0.74 0.3049 0.0148 1>
    Pink=<rgba 0.85 0.255 0.8004 1>
    Purple=<rgba 0.21 0.02 0.4 1>
    R1 Black:LinearColor=<rgba 0.776 0.832 1 1>
    R1 BurntSienna:LinearColor=<rgba 0.776 0.832 1 1>
    R1 Cobalt:LinearColor=<rgba 1 1 1 1>
    R1 Crimson:LinearColor=<rgba 0.8073 0.8555 1 1>
    R1 ForestGreen:LinearColor=<rgba 1 1 1 1>
    R1 Grey:LinearColor=<rgba 1 1 1 1>
    R1 Lime:LinearColor=<rgba 1 1 1 1>
    R1 Orange:LinearColor=<rgba 0.8646 0.8984 1 1>
    R1 Pink:LinearColor=<rgba 0.8646 0.8984 1 1>
    R1 Purple:LinearColor=<rgba 1 1 1 1>
    R1 Saffron:LinearColor=<rgba 1 1 1 1>
    R1 SkyBlue:LinearColor=<rgba 1 1 1 1>
    R1 TitaniumWhite:LinearColor=<rgba 0.8646 0.8984 1 1>
    Saffron=<rgba 0.7 0.7 0.014 1>
    SkyBlue=<rgba 0.098 0.4692 0.7 1>
    TitaniumWhite=<rgba 0.9 0.9 0.9 1>
    Use R1 Painted Overrides=true
   )
   &InitializeParticle(
    Color=$Particles.Color
    Lifetime Max=1.1
    Lifetime Min=0.7
    Lifetime Mode=::NewEnumerator1
    Mesh Scale Mode=::NewEnumerator1
    Mesh Uniform Scale Max=0.75
    Mesh Uniform Scale Min=0.7
    UsePositionOffset=false
    Write Material Random=true
   )
   &PartitionParticles(Number Of Partitions=$Emitter.PortCount)
   &SkeletalMeshLocation(
    Bone / Socket Index=$Particles.PartitionParticles.Partition
    Bone / Socket Sampling Mode=::NewEnumerator5
    MaskSpawnGroup=false
    Mesh Sampling Type=::NewEnumerator6
    Skeletal Mesh=$System.SkeletalMesh
   )
   &ShapeLocation(
    MaskSpawnGroup=false
    Non Uniform Scale=<1 1 0.7>
    Offset=&AddVectorToPosition(Position=<5 0 10>, Vector=$User.DTCarOffset)
    Offset Coordinate Space=::Local
    Offset Mode=::NewEnumerator0
    Override Seed=false
    Sphere Radius=60
    UseOwnerScale=true
   )
   &InitialMeshOrientation(Use Rotation Vector=false)
   &AddVelocity(Velocity=&MergeHelper_VectorSwitch(
     0=&SelectVectorFromArray(
      Direct Array Index=&SpawnGroup()
      Vector Selection Array=DataInterface() {propertyValues {FloatData [<-1400 0 0>, <-1300 0 0>, <-1500 0 0>]}}
     )
     1=&SelectVectorFromArray(
      Direct Array Index=&SpawnGroup()
      Vector Selection Array=DataInterface() {propertyValues {FloatData [<-1100 0 0>, <-1000 0 0>, <-1200 0 0>]}}
     )
     2=&SelectVectorFromArray(
      Direct Array Index=&SpawnGroup()
      Vector Selection Array=DataInterface() {propertyValues {FloatData [<-1000 0 0>, <-900 0 0>, <-1100 0 0>]}}
     )
     3=?
     Index=$User.GameModeID
     Num=::NewEnumerator1
    ))
   /SetVariables() {isSetParameters=true}
   /SetVariables(Particles.breadoffset=-0.1, Particles.sprinklesoffset=0.05) {isSetParameters=true}
  ]
  particleUpdateScript [
   &ParticleState()
   &InheritSourceMovement(ApplyForce=false, ApplyPosition=true, ApplyPositionFalloff=false)
   &UpdateMeshOrientation(Rotation Rate=0.8, Rotation Vector=&RandomVector(Vector Scale=1))
   &GravityForce(Gravity=&MakeVector(
     X=0
     Y=0
     Z=&FloatFromCurve(FloatCurve=DataInterface() {propertyValues {Curve {keys [<at 0.6 0 RCIM_Cubic>, {arriveTangent=-2026.4426, interpMode=RCIM_Cubic, leaveTangent=-2026.4426, tangentMode=RCTM_User, time=1, value=-500}]}}})
    ))
   &SolveForcesAndVelocity(Clamp Velocity=false, Limit Acceleration=false)
   &ScaleMeshSize(Scale Factor=&VectorFromFloat(Value=&FloatFromCurve(FloatCurve=DataInterface() {propertyValues {Curve {keys [
          <at 0 0.05>
          <at 0.12 1>
          {arriveTangent=-1.4286, arriveTangentWeight=0.3202, leaveTangent=-1.4286, leaveTangentWeight=0.1067, time=0.82, value=1}
          {arriveTangent=-8.2212, arriveTangentWeight=0.05, leaveTangent=-8.2212, time=1, value=0}
         ]}}})))
  ]
  rendererClasses [&NiagaraMeshRendererProperties, &NiagaraSpriteRendererProperties]
  renderers [
   {rendererClass=&NiagaraMeshRendererProperties, rendererIndex=0}
   {rendererClass=&NiagaraMeshRendererProperties, rendererIndex=1}
   {rendererClass=&NiagaraMeshRendererProperties, rendererIndex=2}
   {rendererClass=&NiagaraSpriteRendererProperties, rendererIndex=3}
  ]
 }
 NE_VC_Splash {
  emitterUpdateScript [
   &EmitterState(Inactive Response=::NewEnumerator1, Life Cycle Mode=::NewEnumerator1, Loop Duration=1, UseLoopDelay=false)
   &VC_FilteredSocketCount()
   &SpawnRate(Spawn Probability=0.7, SpawnRate=64, Use Spawn Probability=true)
   &SpawnRate(SpawnGroup=1, SpawnRate=12, Use Spawn Probability=false) {module=SpawnRate001}
  ]
  particleSpawnScript [
   &VC_PaintedVariantsOverride(
    Black=&RandomRangeLinearColor(Minimum=<rgba 0 0 0 1>, Maximum=<rgba 0 0 0 1>)
    BurntSienna=&RandomRangeLinearColor(Minimum=<rgba 0.1 0.0145 0.005 1>, Maximum=<rgba 0.15 0.0353 0.015 1>)
    Cobalt=&RandomRangeLinearColor(Minimum=<rgba 0.0306 0.0564 0.34 1>, Maximum=<rgba 0.0365 0.0943 0.73 1>)
    Crimson=&RandomRangeLinearColor(Minimum=<rgba 0.2 0 0 1>, Maximum=<rgba 0.4 0 0 1>)
    DefaultColor=&RandomRangeLinearColor(Minimum=<rgba 0.3124 0.0433 0.3826 1>, Maximum=<rgba 0.8174 0.3127 0.8078 1>)
    ForestGreen=&RandomRangeLinearColor(Minimum=<rgba 0 0.2 0 1>, Maximum=<rgba 0 0.3 0 1>)
    Grey=&RandomRangeLinearColor(Minimum=<rgba 0.1 0.1 0.1 1>, Maximum=<rgba 0.27 0.27 0.27 1>)
    Lime=&RandomRangeLinearColor(Minimum=<rgba 0.2707 0.4 0.012 1>, Maximum=<rgba 0.3467 0.5 0.04 1>)
    Orange=&RandomRangeLinearColor(Minimum=<rgba 0.4 0.1648 0.008 1>, Maximum=<rgba 0.4 0.1696 0.016 1>)
    Pink=&RandomRangeLinearColor(Minimum=<rgba 0.56 0.14 0.525 1>, Maximum=<rgba 0.55 0.1375 0.5156 1>)
    Purple=&RandomRangeLinearColor(Minimum=<rgba 0.1262 0.0025 0.25 1>, Maximum=<rgba 0.11 0.02 0.2 1>)
    Saffron=&RandomRangeLinearColor(Minimum=<rgba 0.4 0.4 0.04 1>, Maximum=<rgba 0.5 0.5 0.05 1>)
    SkyBlue=&RandomRangeLinearColor(Minimum=<rgba 0.007 0.2242 0.35 1>, Maximum=<rgba 0.03 0.391 0.6 1>)
    TitaniumWhite=&RandomRangeLinearColor(Minimum=<rgba 0.6 0.6 0.6 1>, Maximum=<rgba 0.6 0.6 0.6 1>)
   )
   &InitializeParticle(
    Color=$Particles.Color
    Lifetime Max=0.2
    Lifetime Min=0.12
    Lifetime Mode=::NewEnumerator1
    Mesh Scale=&Multiply_Vector(
     A=&RandomRangeVector(Minimum=<1 0.7 0.7>, Maximum=<1.9 1 1>)
     B=&MergeHelper_VectorSwitch(
      0=<0.8 0.8 0.8>
      1=<1 1 1>
      2=?
      3=?
      Index=&MakeCustomIntFromBool(Bool=$User.bIsFrontendPreview, True INT=1, False INT=0)
      Num=::NewEnumerator0
     )
    )
    Mesh Scale Mode=::NewEnumerator4
    Sprite Rotation Mode=::NewEnumerator2
    Sprite Size=&Multiply_Vector2D(
     A=&RandomRangeVector2D(Minimum=<45 200>, Maximum=<100 300>)
     B=&MergeHelper_Vector2Switch(
      0=<0.8 0.9>
      1=<1 1>
      2=?
      3=?
      4=?
      Index=&MakeCustomIntFromBool(Bool=$User.bIsFrontendPreview, True INT=1, False INT=0)
      Num=::NewEnumerator0
     )
    )
    Sprite Size Mode=::NewEnumerator4
    Sprite UV Mode=::NewEnumerator1
    UsePositionOffset=false
    Write Material Random=true
   )
   &PartitionParticles(Number Of Partitions=$Emitter.PortCount)
   &SkeletalMeshLocation(
    Bone / Socket Index=$Particles.PartitionParticles.Partition
    Bone / Socket Sampling Mode=::NewEnumerator5
    MaskSpawnGroup=false
    Mesh Sampling Type=::NewEnumerator6
    Skeletal Mesh=$System.SkeletalMesh
   )
   &ShapeLocation(
    MaskSpawnGroup=false
    Non Uniform Scale=<1 1 1.05>
    Offset=&AddVectorToPosition(Position=<-5 0 0>, Vector=$User.DTCarOffset)
    Offset Coordinate Space=::Local
    Offset Mode=::NewEnumerator0
    Override Seed=false
    Sphere Radius=15
    UseOwnerScale=true
   )
   &InitialMeshOrientation(
    Mesh Orientation Mode=::NewEnumerator7
    Rotation=&RandomRangeVector(Minimum=<0 -0.02 -0.05>, Maximum=<1 0.02 0.05>)
    Use Rotation Vector=true
   )
   &AddVelocity(Velocity=&MergeHelper_VectorSwitch(
     0=&RandomRangeVector(Minimum=<-200 -0.03 -0.05>, Maximum=<-400 0.03 0.05>)
     1=&RandomRangeVector(Minimum=<-2 -0.01 -0.02>, Maximum=<-6 0.01 0.02>)
     2=&RandomRangeVector(Minimum=<-3 -0.03 -0.05>, Maximum=<-9 0.03 0.05>)
     3=?
     Index=$User.GameModeID
     Num=::NewEnumerator1
    ))
  ]
  particleUpdateScript [
   &ParticleState()
   &SolveForcesAndVelocity(Clamp Velocity=false, Limit Acceleration=false)
   &ScaleMeshSize(Scale Factor=&VectorFromCurve(VectorCurve=DataInterface() {propertyValues {
       XCurve {keys [
         {arriveTangent=0.1147, interpMode=RCIM_Cubic, leaveTangent=0.1147, tangentMode=RCTM_User, time=0, value=0.4708}
         {arriveTangent=2.2265, interpMode=RCIM_Cubic, leaveTangent=2.2265, tangentMode=RCTM_User, time=1, value=1.56}
        ]}
       YCurve {keys [<at 0 0.6361>,<at 0.6971 0.9858>]}
       ZCurve {keys [<at 0 0.6361>,<at 0.6971 0.9858>]}
      }}))
   &ScaleSpriteSize(
    Non-Uniform Curve Scale=<1 0.8>
    Non-Uniform Curve Sprite Scale=DataInterface() {propertyValues {
      XCurve {keys [
        {arriveTangent=0.1462, interpMode=RCIM_Cubic, leaveTangent=0.1462, tangentMode=RCTM_User, time=0, value=0.7}
        {arriveTangent=0.0356, interpMode=RCIM_Cubic, leaveTangent=0.0356, tangentMode=RCTM_User, time=1, value=1}
       ]}
      YCurve {keys [
        {arriveTangent=0.1595, interpMode=RCIM_Cubic, leaveTangent=0.1595, tangentMode=RCTM_User, time=0, value=0.6}
        {arriveTangent=0.7323, interpMode=RCIM_Cubic, leaveTangent=0.7323, tangentMode=RCTM_User, time=1, value=1}
       ]}
     }}
    Scale Sprite Size Mode=::NewEnumerator3
   )
   &ScaleColor(
    Scale Alpha=&Multiply_Float(
     A=&FloatFromCurve(FloatCurve=DataInterface() {propertyValues {Curve {
         defaultValue=1
         keys [
          {arriveTangent=-2.1833, interpMode=RCIM_Cubic, leaveTangent=-2.1833, tangentMode=RCTM_User, time=0.8, value=1}
          {arriveTangent=-1.0208, interpMode=RCIM_Cubic, leaveTangent=-1.0208, tangentMode=RCTM_User, time=1, value=0}
         ]
        }}})
     B=8
    )
    Scale RGB=&VectorFromFloat(Value=&RandomRangeFloat(Minimum=0.9, Maximum=1.7))
    ScaleA=true
    ScaleRGB=true
   )
  ]
  rendererClasses [&NiagaraMeshRendererProperties, &NiagaraSpriteRendererProperties]
  renderers [
   {rendererClass=&NiagaraMeshRendererProperties, rendererIndex=0}
   {rendererClass=&NiagaraSpriteRendererProperties, rendererIndex=1}
  ]
 }
 NE_VC_Sprinkles {
  simTarget=GPUComputeSim
  emitterUpdateScript [
   &EmitterState()
   &VC_FilteredSocketCount()
   &SpawnRate(SpawnRate=135, Use Spawn Probability=false)
   &SpawnRate(SpawnGroup=1, SpawnRate=80, Use Spawn Probability=false) {module=SpawnRate001}
  ]
  particleSpawnScript [
   &VC_PaintedVariantsOverride(
    Black=<rgba 0 0 0 1>
    BurntSienna=<rgba 0.21 0.0601 0.0336 1>
    Cobalt=<rgba 0.078 0.1365 0.78 1>
    Crimson=<rgba 0.6 0 0 1>
    DefaultColor=&SelectLinearColorFromArray(
     Array Sampling Mode=::NewEnumerator0
     Direct Array Index=?
     Color Selection Array=DataInterface() {propertyValues {ColorData [<rgba 0.7545 0.7652 0.2994 1>, <rgba 0.8609 0.262 0.643 1>, <rgba 0.262 0.7999 0.8609 1>]}}
     Random Seed=0
    )
    ForestGreen=<rgba 0 0.4 0 1>
    Grey=<rgba 0.24 0.24 0.24 1>
    Lime=<rgba 0.404 0.6 0.012 1>
    Orange=<rgba 0.63 0.2596 0.0126 1>
    Pink=<rgba 0.55 0.2365 0.5239 1>
    Purple=<rgba 0.1855 0.021 0.35 1>
    R1 Black:LinearColor=<rgba 0.02 0.02 0.02 1>
    R1 BurntSienna:LinearColor=<rgba 0.19 0.0705 0.0494 1>
    R1 Cobalt:LinearColor=<rgba 0.1176 0.1778 0.84 1>
    R1 Crimson:LinearColor=<rgba 0.7 0.035 0.035 1>
    R1 DefaultColor=<rgba 0.527 0.2292 0.474 1>
    R1 ForestGreen:LinearColor=<rgba 0.036 0.45 0.0567 1>
    R1 Grey:LinearColor=<rgba 0.22 0.22 0.22 1>
    R1 Lime:LinearColor=<rgba 0.303 0.45 0.009 1>
    R1 Orange:LinearColor=<rgba 0.62 0.2629 0.0248 1>
    R1 Pink:LinearColor=<rgba 0.52 0.2288 0.4957 1>
    R1 Purple:LinearColor=<rgba 0.2012 0.0525 0.35 1>
    R1 Saffron:LinearColor=<rgba 0.5 0.5 0.025 1>
    R1 SkyBlue:LinearColor=<rgba 0.0135 0.2899 0.45 1>
    R1 TitaniumWhite:LinearColor=<rgba 0.6 0.6 0.6 1>
    Saffron=<rgba 0.6 0.6 0.018 1>
    SkyBlue=<rgba 0.0165 0.3544 0.55 1>
    TitaniumWhite=<rgba 0.6 0.6 0.6 1>
    Use R1 Painted Overrides=true
   )
   &InitializeParticle(
    Color=$Particles.Color
    Lifetime Max=1.4
    Lifetime Min=1.2
    Lifetime Mode=::NewEnumerator1
    Sprite Rotation Mode=::NewEnumerator2
    Sprite Size=&Multiply_Vector2D_ByFloat(Float=&RandomRangeFloat(Minimum=0.8, Maximum=1), Vector2D=<11 28>)
    Sprite Size Mode=::NewEnumerator4
    UsePositionOffset=false
    Write Material Random=true
   )
   &PartitionParticles(Number Of Partitions=$Emitter.PortCount)
   &SkeletalMeshLocation(
    Bone / Socket Index=$Particles.PartitionParticles.Partition
    Bone / Socket Sampling Mode=::NewEnumerator5
    MaskSpawnGroup=false
    Mesh Sampling Type=::NewEnumerator6
    Skeletal Mesh=$System.SkeletalMesh
   )
   &ShapeLocation(
    MaskSpawnGroup=false
    Offset=&AddVectorToPosition(Position=<3 0 0>, Vector=$User.DTCarOffset)
    Offset Coordinate Space=::Local
    Offset Mode=::NewEnumerator0
    Override Seed=false
    Sphere Radius=60
    UseOwnerScale=true
   )
   &AddVelocity(Velocity=&MergeHelper_VectorSwitch(
     0=&RandomRangeVector(Minimum=<-1200 0 0>, Maximum=<-1600 0 0>)
     1=&RandomRangeVector(Minimum=<-800 0 0>, Maximum=<-1200 0 0>)
     2=&RandomRangeVector(Minimum=<-700 0 0>, Maximum=<-1100 0 0>)
     3=?
     Index=$User.GameModeID
     Num=::NewEnumerator1
    ))
   /SetVariables(Particles.blobscolor:LinearColor=&MultiplyLinearColor(
     Linear Color=$Particles.R1 Color
     Scale Factor=&MakeVector4FromVectorAndFloat("Vector (RGB)"=&VectorFromFloat(Value=1.7), "Float (Alpha)"=1)
    )) {isSetParameters=true, module=SetVariables_#2}
   /SetVariables(Particles.VisibilityTag=$Emitter.SpawnGroup) {isSetParameters=true, module=SetVariables_#3}
  ]
  particleUpdateScript [
   &ParticleState()
   &InheritSourceMovement(ApplyForce=false, ApplyPosition=true, ApplyPositionFalloff=false)
   &SolveForcesAndVelocity(Clamp Velocity=false, Limit Acceleration=false)
   &ScaleSpriteSize(Uniform Curve Sprite Scale=DataInterface() {propertyValues {Curve {keys [{tangentMode=RCTM_User, time=0, value=0.3}, <at 0.08 1>, <at 0.8 1>, {tangentMode=RCTM_User, time=1, value=0}]}}})
   /SetVariables(Particles.blobssize:Vector2f=&Multiply_Vector2D_ByFloat(
     Float=&FloatFromCurve(FloatCurve=DataInterface() {propertyValues {Curve {keys [{tangentMode=RCTM_User, time=0, value=0.1}, <at 0.1 1>, <at 0.8 1>, {tangentMode=RCTM_User, time=1, value=0}]}}})
     Vector2D=&RandomRangeVector2D(Minimum=<8 8>, Maximum=<17 17>)
    )) {isSetParameters=true, module=SetVariables_#4}
   &SpriteRotationRate(Rotation Rate=&Multiply_Float(A=&RandomRangeFloat(Minimum=-300, Maximum=300), B=0.5))
  ]
  rendererClasses [&NiagaraSpriteRendererProperties]
  renderers [
   {rendererClass=&NiagaraSpriteRendererProperties}
   {rendererClass=&NiagaraSpriteRendererProperties, rendererIndex=1}
   {rendererClass=&NiagaraSpriteRendererProperties, rendererIndex=2}
  ]
 }
 NE_VC_Sprinkles001 {
  emitterUpdateScript [
   &EmitterState()
   &VC_FilteredSocketCount()
   &SpawnRate(SpawnRate=70, Use Spawn Probability=false)
   &SpawnRate(SpawnGroup=1, SpawnRate=80, Use Spawn Probability=false) {module=SpawnRate001}
  ]
  particleSpawnScript [
   &VC_PaintedVariantsOverride(
    Black=<rgba 0 0 0 1>
    BurntSienna=<rgba 0.21 0.0601 0.0336 1>
    Cobalt=<rgba 0.078 0.1365 0.78 1>
    Crimson=<rgba 0.6 0 0 1>
    DefaultColor=&SelectLinearColorFromArray(
     Array Sampling Mode=::NewEnumerator0
     Direct Array Index=?
     Color Selection Array=DataInterface() {propertyValues {ColorData [<rgba 0.7545 0.7652 0.2994 1>, <rgba 0.8609 0.262 0.643 1>, <rgba 0.262 0.7999 0.8609 1>]}}
     Random Seed=0
    )
    ForestGreen=<rgba 0 0.4 0 1>
    Grey=<rgba 0.24 0.24 0.24 1>
    Lime=<rgba 0.404 0.6 0.012 1>
    Orange=<rgba 0.63 0.2596 0.0126 1>
    Pink=<rgba 0.55 0.2365 0.5239 1>
    Purple=<rgba 0.1855 0.021 0.35 1>
    R1 Black:LinearColor=<rgba 0.02 0.02 0.02 1>
    R1 BurntSienna:LinearColor=<rgba 0.19 0.0705 0.0494 1>
    R1 Cobalt:LinearColor=<rgba 0.1176 0.1778 0.84 1>
    R1 Crimson:LinearColor=<rgba 0.7 0.035 0.035 1>
    R1 DefaultColor=<rgba 0.527 0.2292 0.474 1>
    R1 ForestGreen:LinearColor=<rgba 0.036 0.45 0.0567 1>
    R1 Grey:LinearColor=<rgba 0.22 0.22 0.22 1>
    R1 Lime:LinearColor=<rgba 0.303 0.45 0.009 1>
    R1 Orange:LinearColor=<rgba 0.62 0.2629 0.0248 1>
    R1 Pink:LinearColor=<rgba 0.52 0.2288 0.4957 1>
    R1 Purple:LinearColor=<rgba 0.2012 0.0525 0.35 1>
    R1 Saffron:LinearColor=<rgba 0.5 0.5 0.025 1>
    R1 SkyBlue:LinearColor=<rgba 0.0135 0.2899 0.45 1>
    R1 TitaniumWhite:LinearColor=<rgba 0.6 0.6 0.6 1>
    Saffron=<rgba 0.6 0.6 0.018 1>
    SkyBlue=<rgba 0.0165 0.3544 0.55 1>
    TitaniumWhite=<rgba 0.6 0.6 0.6 1>
    Use R1 Painted Overrides=true
   )
   &InitializeParticle(
    Color=$Particles.Color
    Lifetime Max=1.4
    Lifetime Min=1.2
    Lifetime Mode=::NewEnumerator1
    Sprite Rotation Mode=::NewEnumerator2
    Sprite Size=&Multiply_Vector2D_ByFloat(Float=&RandomRangeFloat(Minimum=0.8, Maximum=1), Vector2D=<11 28>)
    Sprite Size Mode=::NewEnumerator4
    UsePositionOffset=false
    Write Material Random=true
   )
   &PartitionParticles(Number Of Partitions=$Emitter.PortCount, Particles Per Partition=3, Partition Mode:ENiagara_PartitionMode=::NewEnumerator1)
   &SkeletalMeshLocation(
    Bone / Socket Index=$Particles.PartitionParticles.Partition
    Bone / Socket Sampling Mode=::NewEnumerator5
    MaskSpawnGroup=false
    Mesh Sampling Type=::NewEnumerator6
    Skeletal Mesh=$System.SkeletalMesh
   )
   &ShapeLocation(
    MaskSpawnGroup=false
    Offset=&AddVectorToPosition(Position=<3 0 0>, Vector=$User.DTCarOffset)
    Offset Coordinate Space=::Local
    Offset Mode=::NewEnumerator0
    Override Seed=false
    Sphere Radius=60
    UseOwnerScale=true
   )
   &AddVelocity(Velocity=&MergeHelper_VectorSwitch(
     0=&RandomRangeVector(Minimum=<-1200 0 0>, Maximum=<-1600 0 0>)
     1=&RandomRangeVector(Minimum=<-800 0 0>, Maximum=<-1200 0 0>)
     2=&RandomRangeVector(Minimum=<-700 0 0>, Maximum=<-1100 0 0>)
     3=?
     Index=$User.GameModeID
     Num=::NewEnumerator1
    ))
   /SetVariables(Particles.blobscolor:LinearColor=&MultiplyLinearColor(
     Linear Color=$Particles.R1 Color
     Scale Factor=&MakeVector4FromVectorAndFloat("Vector (RGB)"=&VectorFromFloat(Value=1.7), "Float (Alpha)"=1)
    )) {isSetParameters=true, module=SetVariables_#2}
   /SetVariables(Particles.VisibilityTag=$Emitter.SpawnGroup) {isSetParameters=true, module=SetVariables_#3}
  ]
  particleUpdateScript [
   &ParticleState()
   &InheritSourceMovement(ApplyForce=false, ApplyPosition=true, ApplyPositionFalloff=false)
   &SolveForcesAndVelocity(Clamp Velocity=false, Limit Acceleration=false)
   &ScaleSpriteSize(Uniform Curve Sprite Scale=DataInterface() {propertyValues {Curve {keys [{tangentMode=RCTM_User, time=0, value=0.3}, <at 0.08 1>, <at 0.8 1>, {tangentMode=RCTM_User, time=1, value=0}]}}})
   /SetVariables(Particles.blobssize:Vector2f=&Multiply_Vector2D_ByFloat(
     Float=&FloatFromCurve(FloatCurve=DataInterface() {propertyValues {Curve {keys [{tangentMode=RCTM_User, time=0, value=0.1}, <at 0.1 1>, <at 0.8 1>, {tangentMode=RCTM_User, time=1, value=0}]}}})
     Vector2D=&RandomRangeVector2D(Minimum=<8 8>, Maximum=<17 17>)
    )) {isSetParameters=true, module=SetVariables_#4}
   &SpriteRotationRate(Rotation Rate=&Multiply_Float(A=&RandomRangeFloat(Minimum=-300, Maximum=300), B=0.5))
  ]
  rendererClasses [&NiagaraSpriteRendererProperties]
  renderers [
   {rendererClass=&NiagaraSpriteRendererProperties}
   {rendererClass=&NiagaraSpriteRendererProperties, rendererIndex=1}
   {rendererClass=&NiagaraSpriteRendererProperties, rendererIndex=2}
  ]
 }
}
summary {
 systemName=NS_Doughnut
 emitters [
  {emitterName=NE_VC_Splash, rendererClasses [&NiagaraMeshRendererProperties, &NiagaraSpriteRendererProperties]}
  {emitterName=NE_VC_Sprinkles, simTarget=GPUComputeSim, rendererClasses [&NiagaraSpriteRendererProperties]}
  {emitterName=NE_VC_Donuts, rendererClasses [&NiagaraMeshRendererProperties, &NiagaraSpriteRendererProperties]}
  {emitterName=NE_VC_Sprinkles001, rendererClasses [&NiagaraSpriteRendererProperties]}
 ]
 userVariables [
  User.bIsFrontendPreview {type=NiagaraBool}
  User.bIsPainted {type=NiagaraBool}
  User.ColorVariantID {type=NiagaraInt32}
  User.DTCarOffset {type=Vector3f, description="", defaultValue {struct=Vector3f, value=<0 0 0>}}
  User.GameModeID {type=NiagaraInt32, defaultValue=2}
  User.PaintedColor {type=LinearColor, description="", defaultValue {struct=LinearColor, value=<rgba 1 1 1 1>}}
  User.PaintedScalar {type=NiagaraFloat, defaultValue=1, description=""}
  User.PortCount {type=NiagaraInt32, defaultValue=2}
 ]
}
